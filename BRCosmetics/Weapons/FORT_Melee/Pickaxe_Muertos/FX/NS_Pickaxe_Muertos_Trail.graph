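# Niagara system for the Muertos pickaxe trail, combining an anim-trail ribbon with a burst of "Notes" sprites.
# Main_Trail feeds three ribbon renderers, shifting warm to blue over life and computing a cosine-waved Z position in a separate New Position attribute.
# Notes spawn by distance moved at a sampled socket, then spin and drift on curl noise through a similar color ramp.
# Time-of-day scalars adjust brightness or opacity, and front-end modifiers shrink both effects.
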NiagaraSystem /BRCosmetics/Weapons/FORT_Melee/Pickaxe_Muertos/FX/NS_Pickaxe_Muertos_Trail
dependencies {
 usedDataInterfaces [&NiagaraDataInterfaceColorCurve, &NiagaraDataInterfaceCurve, &NiagaraDataInterfaceSocketReader]
 usedDynamicInputs []
 usedModules [
  &self:SystemSpawnScript.SystemScriptSource.SystemScriptGraph.NiagaraNodeAssignment_1.SetVariables_#1
  &SystemState
 ]
}
emitters {
 Main_Trail {
  emitterSpawnScript [&FN_RibbonIDOffsetUtility()]
  emitterUpdateScript [
   &EmitterState(Life Cycle Mode=::NewEnumerator1, Loop Behavior=::NewEnumerator1, Loop Delay=0.022, Loop Duration=0.2, UseLoopDelay=true)
   &FN_SpawnAnimTrail()
  ]
  particleSpawnScript [
   &FN_InitializeAnimTrailBoneIndecies()
   &FN_InitializeAnimTrail(
    AnimTrailWidth=&Multiply_Float(A=0.6, B=&FN_FrontEndFloatModifier(Gameplay=1, Front End=0.6))
    Lifetime=0.23
   )
  ]
  particleUpdateScript [
   &SolveForcesAndVelocity(Clamp Velocity=false, Limit Acceleration=false)
   &FN_UpdateAnimTrail() {module=FN_UpdateAnimTrail001}
   &ScaleColor(
    Scale Alpha=&FloatFromCurve(FloatCurve=DataInterface() {propertyValues {Curve {keys [{interpMode=RCIM_Cubic, tangentMode=RCTM_User, time=0.6, value=2}, {interpMode=RCIM_Cubic, tangentMode=RCTM_User, time=1, value=0}]}}})
    Scale RGB=&VectorFromFloat(Value=&FloatFromCurve(
      FloatCurve=DataInterface() {propertyValues {Curve {keys [
          {arriveTangent=-17.2503, interpMode=RCIM_Cubic, leaveTangent=-17.2503, tangentMode=RCTM_User, time=0, value=8}
          <at 0.25 1.5>
          {interpMode=RCIM_Cubic, tangentMode=RCTM_User, time=1, value=0.2}
         ]}}}
      Scale Curve=&FNTimeOfDayScalar(Night Scale=0.5, Morning Scale=0.3, Evening Scale=0.5)
     ))
    ScaleA=true
    ScaleRGB=true
   )
   /SetVariables(Particles.Mask Color:LinearColor=/Niagara/DynamicInputs/LinearColor/MakeLinearColorFromVector4(Vector:Vector4f=/Niagara/DynamicInputs/Multiply/Multiply_Vector4_ByFloat(
      Vector4:Vector4f=&MakeVector4_FromLinearColor(LinearColor=&ColorFromCurve(ColorCurve=DataInterface() {propertyValues {
          AlphaCurve {keys [<at 0.7221 1 RCIM_Cubic>,<at 0.9869 0>]}
          BlueCurve {keys [
            {arriveTangent=1.1111, interpMode=RCIM_Cubic, time=0.1006, value=0}
            {arriveTangent=1.3568, interpMode=RCIM_Cubic, leaveTangent=1.3568, time=0.3335, value=0}
            <at 0.5943 1>
            <at 0.8377 1>
           ]}
          GreenCurve {keys [
            <at 0.1006 0.6051 RCIM_Cubic>
            {arriveTangent=-0.708, interpMode=RCIM_Cubic, leaveTangent=-0.708, time=0.3335, value=0}
            <at 0.5943 0>
            <at 0.8377 0.0833>
           ]}
          RedCurve {keys [
            {arriveTangent=-0.8423, interpMode=RCIM_Cubic, time=0.1006, value=1}
            {arriveTangent=-1.3568, interpMode=RCIM_Cubic, leaveTangent=-1.3568, time=0.3335, value=1}
            <at 0.5943 0.2419>
            <at 0.8377 0>
           ]}
         }}))
      Float:NiagaraFloat=&FloatFromCurve(FloatCurve=DataInterface() {propertyValues {Curve {keys [<at 0.3 1 RCIM_Cubic>,<at 0.9 0 RCIM_Cubic>]}}}, Scale Curve=0.5)
     ))) {isSetParameters=true}
   &ScaleRibbonWidth(Ribbon Width Scale=&FloatFromCurve(FloatCurve=DataInterface() {propertyValues {Curve {keys [
         {arriveTangent=-0.1806, interpMode=RCIM_Cubic, leaveTangent=-0.1806, tangentMode=RCTM_User, time=0, value=0.5}
         {arriveTangent=0.3103, interpMode=RCIM_Cubic, leaveTangent=0.3103, tangentMode=RCTM_User, time=1, value=1.05}
        ]}}}))
   /SetVariables(Particles.New Position:NiagaraPosition=&Add_Vector(
     A=Linked() {linkedVariable {name=Particles.Position, type=NiagaraPosition}}
     B=&MakeVector(
      X=0
      Y=0
      Z=&Cosine(
       Normalized Angle=&Add_Float(
        A=&self:UniqueID()
        B=&Multiply_Float(A=$Emitter.Age, B=-0.4)
       )
       Period=0.1136
       Scale=15
       Bias=0
      )
     )
    )) {isSetParameters=true}
   &DynamicMaterialParameters(
    Index 0 Param 1=0.9
    Index 3 Param 1=0.8
    Index 3 Param 2=0.5
    Module.Index0Float0Write=true
    Module.Index0Float1Write=true
    Module.Index0Float2Write=true
    Module.Index0Float3Write=true
    Module.Index2Float0Write=true
    Module.Index2Float1Write=true
    Module.Index2Float2Write=true
    Module.Index2Float3Write=true
    Module.Index3Float0Write=true
    Module.Index3Float1Write=true
    Module.Index3Float2Write=true
    Module.Index3Float3Write=true
    Write Parameter Index 2=::NewEnumerator1
    Write Parameter Index 3=::NewEnumerator1
   )
  ]
  rendererClasses [&NiagaraRibbonRendererProperties]
  renderers [
   {rendererClass=&NiagaraRibbonRendererProperties}
   {rendererClass=&NiagaraRibbonRendererProperties, rendererIndex=1}
   {rendererClass=&NiagaraRibbonRendererProperties, rendererIndex=2}
  ]
 }
 Notes {
  emitterUpdateScript [
   &EmitterState(Life Cycle Mode=::NewEnumerator1, Loop Behavior=::NewEnumerator1, Loop Delay=0.01, Loop Duration=0.2, UseLoopDelay=true)
   &SpawnPerUnit(Module.Use Max Movement Threshold=true, Module.Use Movement Tolerance=false, Spawn Per Unit=50, Use Spawn Probability=false)
  ]
  particleSpawnScript [
   &FN_SetSpawnGroup(Number of Spawn Groups=2)
   &InitializeParticle(
    Color=&RandomRangeLinearColor(Minimum=<rgba 0.1 0.1 0.1 2>, Maximum=<rgba 2 2 2 1>)
    Lifetime Max=0.6
    Lifetime Min=0.3
    Lifetime Mode=::NewEnumerator1
    Sprite Rotation Mode=::NewEnumerator2
    Sprite Size Mode=::NewEnumerator3
    Sprite UV Mode=::NewEnumerator3
    Uniform Sprite Size=&Multiply_Float(
     A=&RandomRangeFloat(Minimum=15, Maximum=35)
     B=&FN_FrontEndFloatModifier(Gameplay=1, Front End=0.6)
    )
    UsePositionOffset=false
    Write Material Random=true
   )
   &SubUVAnimation(End Frame=3, SubUV Animation Mode:ENiagara_SubUVLookupMode=::NewEnumerator2)
   &SocketLocation(
    Interpolate Socket Sampling=true
    SamplePosOffset=true
    Sampled Position Offset=<0 0 -3>
    Socket Index=1
    Socket Reader=$System.Socket Reader
    Socket Sampling Mode=::NewEnumerator3
    VelocitySpeedLimit=false
   )
  ]
  particleUpdateScript [
   &ParticleState()
   &AccelerationForce(Acceleration=&RandomRangeVector(Minimum=<0 0 200>, Maximum=<0 0 400>))
   &GravityForce(Gravity=&RandomRangeVector(Minimum=<0 0 100>, Maximum=<0 0 300>))
   &CurlNoiseForce(Noise Strength=&RandomRangeFloat(Minimum=500, Maximum=700), Pan Noise=false)
   &Drag(Drag=&RandomRangeFloat(Minimum=4, Maximum=6), Use Linear Drag=true, Use Rotational Drag=false)
   &SolveForcesAndVelocity(Clamp Velocity=false, Limit Acceleration=false)
   &ScaleColor(
    Scale Alpha=&FNTimeOfDayScalar(Night Scale=0.7, Morning Scale=0.5, Evening Scale=0.5)
    Scale RGB=&MakeVectorFromLinearColorRGB(Color=&ColorFromCurve(ColorCurve=DataInterface() {propertyValues {
        AlphaCurve {keys [<at 0.6 1 RCIM_Cubic>,<at 1 0>]}
        BlueCurve {keys [
          {arriveTangent=1.1111, interpMode=RCIM_Cubic, time=0.1266, value=0}
          {arriveTangent=1.1449, interpMode=RCIM_Cubic, leaveTangent=1.1449, time=0.3797, value=0}
          <at 0.6928 0.6795>
          <at 1 1>
         ]}
        GreenCurve {keys [
          <at 0.1266 1 RCIM_Cubic>
          {arriveTangent=-1.0728, interpMode=RCIM_Cubic, leaveTangent=-1.0728, time=0.3797, value=0}
          <at 0.6928 0>
          <at 1 0.063>
         ]}
        RedCurve {keys [
          {arriveTangent=-0.8423, interpMode=RCIM_Cubic, time=0.1266, value=1}
          {arriveTangent=-1.1449, interpMode=RCIM_Cubic, leaveTangent=-1.1449, time=0.3797, value=1}
          <at 0.6928 0.0908>
          <at 1 0>
         ]}
       }}))
    ScaleA=true
    ScaleRGB=true
   )
   &SpriteRotationRate(Rotation Rate=&RandomRangeFloat(Minimum=180, Maximum=280))
   &ScaleSpriteSize(Scale Factor=&Vector2DFromFloat(Value=&FloatFromCurve(FloatCurve=DataInterface() {propertyValues {Curve {keys [
          {arriveTangent=-1.0014, interpMode=RCIM_Cubic, leaveTangent=-1.0014, tangentMode=RCTM_User, time=0, value=0}
          {arriveTangentWeight=0.1156, interpMode=RCIM_Cubic, leaveTangentWeight=0.2312, tangentMode=RCTM_User, time=0.2, value=1}
          {arriveTangent=-2.3218, interpMode=RCIM_Cubic, leaveTangent=-2.3218, tangentMode=RCTM_User, time=0.6, value=0.65}
          <at 0.9 0>
         ]}}})))
  ]
  rendererClasses [&NiagaraSpriteRendererProperties]
  renderers [{rendererClass=&NiagaraSpriteRendererProperties}]
 }
}
summary {
 systemName=NS_Pickaxe_Muertos_Trail
 emitters [
  {emitterName=Notes, rendererClasses [&NiagaraSpriteRendererProperties]}
  {emitterName=Main_Trail, rendererClasses [&NiagaraRibbonRendererProperties]}
 ]
 userVariables [
  User.bIsFrontEnd {type=NiagaraBool}
  User.bIsFrontEndPreview {type=NiagaraBool}
  User.RibbonIDOffset {type=NiagaraInt32}
 ]
}
